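# Masked, two-sided crow material for mesh particles that animates the bird through world position offset.
# UV channels 1 and 2 mask legs, body and wings, driving sine translations and pivot rotations for body bob, leg and tail motion, and delayed-tip flapping.
# Dynamic parameters set speed, offset flap timing and blend flapping into a slower glide motion.
# An Emissive parameter shifts the textured lit look toward desaturated, particle-colored glow.
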Material /Game/Effects/Fort_Effects/Materials/Birds/M_Crow_WPO
graph {
 properties {
  bTangentSpaceNormal=false
  bUsedWithMeshParticles=true
  bUsedWithNiagaraMeshParticles=true
  blendMode=BLEND_Masked
  opacityMaskClipValue=0.333
  pixelDepthOffsetMode=PDOM_Legacy
  previewMesh=/Game/Effects/Fort_Effects/Mesh/Crows/SM_Crow_WPO
  twoSided=true
  usedShadingModels=Default Lit
 }
 nodes [
  Constant() {r=0.5}
  n1 = TextureSample() {texture=/Game/Effects/Fort_Effects/Textures/Crows/T_crow_01}
  n2 = Multiply(A=@n1.A, B=Multiply(A=Power(Base=@n1.R)) {constB=50})
  n3 = MaterialFunctionCall(DiffuseColor=@n1.RGB, SpecularColor=Multiply(A=Clamp(None=@n2) {maxDefault=3}, B=Constant3Vector() {constant=<rgba 0.0811 0.1199 0.35 0>})) {materialFunction=&ConvertFromDiffSpec}
  n4 = DynamicParameter() {defaultValue=<rgba 0 1.6 0 1>, paramNames [Glide,WingSpeed,AnimOffset,Emissive]}
  LinearInterpolate(A=@n3.BaseColor, B=ScalarParameter() {defaultValue=0.1, parameterName=Emissive_BaseColor}, Alpha=@n4.Emissive) {constB=0.1}
  LinearInterpolate(B=Multiply(A=Desaturation(None=@n3.BaseColor), B=ParticleColor().RGB), Alpha=@n4.Emissive)
  LinearInterpolate(A=@n3.Metallic, B=Multiply(A=@n3.Metallic) {constB=0.1}, Alpha=@n4.Emissive) {constB=0.3}
  LinearInterpolate(A=@n3.Specular, B=Multiply(A=@n3.Specular) {constB=0.1}, Alpha=@n4.Emissive) {constB=0.3}
  n5 = Transform(None=Constant3Vector() {constant=<rgba 1 0 0.5 0>}) {transformSourceType=TRANSFORMSOURCE_Local}
  n6 = Add(A=Time(), B=DynamicParameter() {parameterIndex=1}.Param3)
  n7 = DynamicParameter() {defaultValue=<rgba 0 1.6 0 1>, paramNames [Glide,WingSpeed,AnimOffset,Emissive]}
  n8 = Multiply(A=@n6, B=Multiply(A=Constant() {desc=Body Anim Speed, r=1}, B=@n7.WingSpeed))
  n9 = Multiply(A=Multiply(A=Normalize(VectorInput=@n5), B=Sine(None=@n8)), B=Constant() {desc=Body Anim Amount, r=3})
  n10 = Clamp(None=ComponentMask(None=TextureCoordinate() {coordinateIndex=1}) {r=true}) {desc=Fasten Mask}
  n11 = Floor(None=Add(A=TextureCoordinate() {coordinateIndex=2}, B=Constant() {r=0.5}))
  n12 = Clamp(None=Multiply(A=ComponentMask(None=@n11) {r=true}, B=ComponentMask(None=@n11) {g=true})) {desc=Legs}
  n13 = Clamp(None=Add(A=OneMinus(None=@n10), B=@n12))
  n14 = Normalize(VectorInput=Transform(None=Constant3Vector() {constant=<rgba -0.5 0 1 0>}) {transformSourceType=TRANSFORMSOURCE_Local})
  n15 = Constant() {r=0.2}
  n16 = Constant() {desc=Leg Anim Speed, r=1}
  n17 = Sine(None=Multiply(A=Add(A=@n6, B=@n15), B=Multiply(A=@n16, B=@n7.WingSpeed)))
  n18 = Multiply(A=Multiply(A=Multiply(A=@n14, B=@n17), B=Constant() {desc=Leg Anim Amount, r=2}), B=Multiply(A=@n10, B=@n12))
  n19 = Transform(None=Constant3Vector() {constant=<rgba 0 0 1 0>}) {transformSourceType=TRANSFORMSOURCE_Local}
  n20 = Multiply(A=Multiply(A=Normalize(VectorInput=@n19), B=Constant() {desc=Tail Move Amount, r=2}), B=@n17)
  n21 = OneMinus(None=@n11)
  n22 = Clamp(None=Multiply(A=ComponentMask(None=@n21) {r=true}, B=ComponentMask(None=@n21) {g=true})) {desc=Body}
  n23 = Multiply(A=@n10, B=@n22)
  n24 = Normalize(VectorInput=Transform(None=Constant3Vector() {constant=<rgba 1 0 0 0>}) {transformSourceType=TRANSFORMSOURCE_Local})
  n25 = DynamicParameter() {defaultValue=<rgba 0 1.6 0 1>, paramNames [Glide,WingSpeed,AnimOffset,Emissive]}
  n26 = Multiply(A=@n6, B=Multiply(A=Constant() {desc=Tail Rot Speed, r=0.7}, B=@n25.WingSpeed))
  n27 = Multiply(A=Multiply(A=Sine(None=@n26), B=Constant() {desc=Tail Rot Amount, r=0.025}), B=@n23)
  n28 = MaterialFunctionCall() {materialFunction=&ComponentPivotLocation}
  n29 = WorldPosition()
  n30 = TextureCoordinate() {coordinateIndex=1}
  n31 = Constant() {r=0.5}
  n32 = ComponentMask(None=Floor(None=Add(A=@n30, B=@n31))) {g=true}
  n33 = Transform(None=LinearInterpolate(
    A=Constant3Vector() {constant=<rgba 1 0.25 -0.5 0>}
    B=Constant3Vector() {constant=<rgba -1 -0.25 0.5 0>}
    Alpha=Clamp(None=@n32) {desc=Wings L R}
   )) {transformSourceType=TRANSFORMSOURCE_Local}
  n34 = Constant() {r=0.2}
  n35 = Clamp(None=Multiply(A=ComponentMask(None=@n21) {r=true}, B=ComponentMask(None=@n11) {g=true})) {desc=Wings}
  n36 = Multiply(A=Clamp(None=Add(A=@n10, B=@n34)), B=@n35)
  n37 = Add(A=@n6, B=Multiply(A=Constant() {desc=Time Delay for Wing Top, r=0.15}, B=OneMinus(None=@n10)))
  n38 = Multiply(A=@n25.WingSpeed, B=Constant() {desc=Wing Anim Speed, r=1})
  n39 = Multiply(A=Sine(None=Multiply(A=@n37, B=@n38)), B=Constant() {desc=Wing Anim Amount, r=0.1})
  n40 = Add(
   A=Add(A=Multiply(A=@n20, B=@n23), B=RotateAboutAxis(NormalizedRotationAxis=@n24, RotationAngle=@n27, PivotPoint=@n28.Result, Position=@n29.XYZ))
   B=RotateAboutAxis(
    NormalizedRotationAxis=Normalize(VectorInput=@n33)
    RotationAngle=Multiply(A=@n36, B=@n39)
    PivotPoint=MaterialFunctionCall() {materialFunction=&ComponentPivotLocation}.Result
    Position=WorldPosition().XYZ
   )
  )
  n41 = Constant3Vector() {constant=<rgba 0.25 1 0.25 0>}
  n42 = Constant() {desc=Wing Anim Speed, r=0.6}
  n43 = DynamicParameter() {defaultValue=<rgba 0 1.6 0 1>, paramNames [Glide,WingSpeed,AnimOffset,Emissive]}
  n44 = Sine(None=Multiply(A=Time(), B=Multiply(A=@n42, B=@n43.WingSpeed)))
  n45 = Constant() {desc=Wing Anim Amount, r=0.025}
  n46 = Multiply(A=Constant() {r=0.2}, B=@n22)
  n47 = RotateAboutAxis(
   NormalizedRotationAxis=Normalize(VectorInput=Transform(None=@n41) {transformSourceType=TRANSFORMSOURCE_Local})
   RotationAngle=Multiply(A=Multiply(A=@n44, B=@n45), B=Add(A=@n35, B=@n46))
   PivotPoint=MaterialFunctionCall() {materialFunction=&ComponentPivotLocation}.Result
   Position=WorldPosition().XYZ
  )
  n48 = Constant3Vector() {constant=<rgba 0 0 1 0>}
  n49 = Add(A=Time(), B=Multiply(A=Constant() {r=-0.1}, B=@n10))
  n50 = Multiply(A=Constant() {desc=Body Anim Speed, r=0.5}, B=@n43.WingSpeed)
  n51 = Multiply(A=Normalize(VectorInput=Transform(None=@n48) {transformSourceType=TRANSFORMSOURCE_Local}), B=Sine(None=Multiply(A=@n49, B=@n50)))
  n52 = Constant() {desc=Body Anim Amount, r=6}
  LinearInterpolate(A=Add(A=Multiply(A=@n9, B=@n13), B=Add(A=@n18, B=@n40)), B=Add(A=Multiply(A=@n47, B=@n10), B=Multiply(A=@n51, B=@n52)), Alpha=@n4.Glide)
 ]
}
